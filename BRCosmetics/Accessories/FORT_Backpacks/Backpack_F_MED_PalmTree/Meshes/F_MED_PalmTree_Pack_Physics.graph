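# Physics asset for the skeletal mesh of the palm tree backpack cosmetic (F_MED_PalmTree_Pack).
# It has two small capsule bodies, one on the base_clip bone and a larger one on the dyn_clip bone.
# A single constraint joins dyn_clip to base_clip, with a wide limit angle on the second swing axis and zero on the other swing and twist, suggesting a hinge-like dangling clip.
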PhysicsAsset /BRCosmetics/Accessories/FORT_Backpacks/Backpack_F_MED_PalmTree/Meshes/F_MED_PalmTree_Pack_Physics
bodies {
 base_clip [{shapeType=Capsule, length=0.505, radius=0.505, shapeName=base_clip_capsule, rotation=<rot 0 -90 0>}]
 dyn_clip [{shapeType=Capsule, length=2, radius=1, shapeName=dyn_clip_capsule, center=<4 0 0>, rotation=<rot 90 0 0>}]
}
constraints [{bone1Name=dyn_clip, bone2Name=base_clip, swing1LimitDegrees=0, swing2LimitDegrees=110, twistLimitDegrees=0}]
